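; Extract_Bounds is a PCG primitive graph that, going by its name, extracts bounds from a primary input of any type, with a flag to combine all inputs into a single point (on by default).
; It tags the input "Input:0", then chains inputs management, bounds core process and output preparation subgraphs.
; Fallback, input processing (convert to points by default), core and output stages are graph parameters, making each swappable.
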(graph Extract_Bounds
  (params (InputsProcessingGraph :type (SoftObject /Script/PCG.PCGGraphInterface) :default /PCGPrimitives/Primitives/Shared/InputsProcessing/InputsProcessing_ConvertToPoints.InputsProcessing_ConvertToPoints) (InputsFallbackGraph :type (SoftObject /Script/PCG.PCGGraphInterface) :default /PCGPrimitives/Primitives/_Template_/Primitives_InputsFallback.Primitives_InputsFallback) (CoreProcessGraph :type (SoftObject /Script/PCG.PCGGraphInterface) :default /PCGPrimitives/Primitives/Extract/Subgraphs/Extract_Bounds_CoreProcess.Extract_Bounds_CoreProcess) (CombineAllInputsIntoSinglePoint :type Bool :default true) (PrepareOutputsGraph :type (SoftObject /Script/PCG.PCGGraphInterface) :default /PCGPrimitives/Primitives/_Template_/Primitives_PrepareOutputs.Primitives_PrepareOutputs))
  (inputs (PrimaryInput :type Any))
  (bind AddTags_1 (Add%20Tags :TagsToAdd "Input:0"))
  ;[0 0 0 0.5 32] Inputs
  (bind Subgraph_0 (Subgraph :Graph /PCGPrimitives/Primitives/_Template_/Subgraphs/Primitives_InputsManagement.Primitives_InputsManagement :InputsFallbackGraph InputsFallbackGraph :InputsProcessingGraph InputsProcessingGraph))
  ;[0 0 0 0.5 32] Core Process
  (bind Subgraph_3 (Subgraph :Graph /PCGPrimitives/Primitives/Extract/Subgraphs/Extract_Bounds_CoreProcess.Extract_Bounds_CoreProcess :CombineAllInputsIntoSinglePoint CombineAllInputsIntoSinglePoint :SubgraphOverride-param CoreProcessGraph))
  ;[0 0 0 0.5 32] Prepare Output
  (bind Subgraph_4 (Subgraph :Graph /PCGPrimitives/Primitives/_Template_/Primitives_PrepareOutputs.Primitives_PrepareOutputs :SubgraphOverride-param PrepareOutputsGraph))
  (connect DefaultInputNode PrimaryInput AddTags_1 In)
  (connect AddTags_1 Out Subgraph_0 In)
  (connect Subgraph_0 Out Subgraph_3 In)
  (connect Subgraph_3 Out Subgraph_4 In)
  (output Out Subgraph_4))
